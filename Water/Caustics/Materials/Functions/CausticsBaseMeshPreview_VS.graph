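# Material function that outputs World Position Offset to tweak water depth on a preview mesh for caustics textures from the GenerateCausticsTextures Blueprint.
# Vertices are remapped in Z by bounding-box height, from WaterDepth below the actor position to Top Bias above it, scaled by EnableDepthVS.
# A second offset pushes vertices along Y by centered green vertex color times SlopeAdjust, and both offsets are summed.
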MaterialFunction /Water/Caustics/Materials/Functions/CausticsBaseMeshPreview_VS
graph {
 properties {description=This function just adjusts the vertex shader of a mesh so the water depth can be tweaked when previewing caustics textures via the GenerateCausticsTextures BP.}
 nodes [
  n1 = Multiply(A=Subtract(A=VertexColor().G) {constB=0.5}, B=Constant3Vector() {constant=<rgba 0 1 0 0>})
  n2 = ScalarParameter() {parameterName=SlopeAdjust}
  n3 = Subtract(A=ActorPositionWS(), B=WorldPosition().XYZ)
  n4 = ScalarParameter() {defaultValue=512, parameterName=WaterDepth}
  n5 = Multiply(A=ScalarParameter() {defaultValue=64, parameterName=Top Bias}) {constB=-1}
  n6 = MaterialFunctionCall() {materialFunction=&BoundingBoxBased_0-1_UVW}
  n7 = Multiply(
   A=Subtract(A=ComponentMask(None=@n3) {b=true}, B=LinearInterpolate(A=@n4, B=@n5, Alpha=@n6.B)) {constA=64}
   B=Constant3Vector() {constant=<rgba 0 0 1 0>}
  )
  n8 = ScalarParameter() {parameterName=EnableDepthVS}
  FunctionOutput(None=Add(A=Multiply(A=@n1, B=@n2), B=Multiply(A=@n7, B=@n8))) {outputName=WPO}
 ]
}
