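# Material function that makes surfaces look wet near and below a water level, returning modified material attributes and the wetness Mask.
# The mask fades with height above Water Z, which defaults to the WaterBodyData height but accepts a manual override, and Additional Wet Mask pushes it toward fully wet.
# Wetness boosts saturation via negative desaturation, scales base color and blends specular and roughness.
# Deeper underwater, roughness blends to a replacement value.
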MaterialFunction /Water/Materials/Functions/Wetness
graph {nodes [
  n1 = GetMaterialAttributes(None=FunctionInput() {inputType=FunctionInput_MaterialAttributes}) {attributeGetTypes [#1, #2, #3]}
  n2 = ScalarParameter() {defaultValue=0.25, parameterName=Wet Saturation}
  n3 = FunctionInput(Preview=MaterialFunctionCall!(Position) {materialFunction=&/WaterBodyData}.Water Z Height) {
   bUsePreviewValueAsDefault=true
   description="By default, Water Z comes from the WaterBodyData function, which will get the water Z height from the water system. You can override this input to specify a manual water level."
   inputName=Water Z
   inputType=FunctionInput_Scalar
  }
  n4 = ComponentMask(None=Subtract(A=WorldPosition().XYZ, B=@n3)) {b=true}
  n5 = Divide(A=Add(A=ScalarParameter() {parameterName=Wetness Bias}, B=@n4), B=ScalarParameter() {defaultValue=8, parameterName=Wetness Length})
  n6 = FunctionInput() {bUsePreviewValueAsDefault=true, description="0 = Dry, 1 = Wet", inputName=Additional Wet Mask, inputType=FunctionInput_Scalar}
  n7 = LinearInterpolate(A=OneMinus(None=Saturate(None=@n5)), Alpha=Saturate(None=@n6))
  n8 = Desaturation(None=@n1.BaseColor, Fraction=LinearInterpolate(B=Multiply(A=@n2) {constB=-1}, Alpha=@n7) {constB=0})
  n9 = LinearInterpolate(B=ScalarParameter() {defaultValue=0.5, parameterName=Wetness Basecolor Multiplier}, Alpha=@n7) {constA=1}
  n10 = ScalarParameter() {defaultValue=0.225, parameterName=Wet Specular}
  n11 = ScalarParameter() {defaultValue=0.75, parameterName=Underwater Roughness Replace}
  n12 = LinearInterpolate(A=@n1.Roughness, B=ScalarParameter() {defaultValue=0.03, parameterName=Wet Roughness}, Alpha=@n7)
  n13 = ScalarParameter() {defaultValue=16, parameterName=Underwater Rough Removal Bias}
  n14 = Saturate(None=Divide(A=Add(A=@n13, B=@n4), B=ScalarParameter() {defaultValue=16, parameterName=Underwater Roughness Spread}))
  FunctionOutput(None=SetMaterialAttributes!(
    MaterialAttributes=@n1.MaterialAttributes
    BaseColor=Multiply(A=@n8, B=@n9)
    Specular=LinearInterpolate(A=@n1.Specular, B=@n10, Alpha=@n7)
    Roughness=LinearInterpolate(A=@n11, B=@n12, Alpha=@n14) {constA=1}
   ) {attributeSetTypes [#1, #2, #3]})
  FunctionOutput(None=@n7) {outputName=Mask, sortPriority=1}
 ]}
