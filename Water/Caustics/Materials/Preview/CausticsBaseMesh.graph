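# Opaque lit preview material for the base mesh in the Water plugin's caustics preview.
# Base color is a squared BaseColor tint times a world-aligned engine tile texture, with full roughness and the world-space vertex normal.
# CausticsBaseMeshPreview_VS supplies World Position Offset, and the assembled material attributes pass through the Wetness function with its water height and extra mask inputs unconnected.
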Material /Water/Caustics/Materials/Preview/CausticsBaseMesh
graph {
 properties {bTangentSpaceNormal=false, bUseMaterialAttributes=true, pixelDepthOffsetMode=PDOM_Legacy, usedShadingModels=Default Lit}
 nodes [
  n1 = VectorParameter() {parameterName=BaseColor, defaultValue=<rgba 1.2 1 1 2>}
  n2 = TextureObject() {texture=/Engine/EditorResources/TilePatine_D}
  n3 = ScalarParameter() {defaultValue=1024, parameterName=Texture Size}
  n4 = SetMaterialAttributes!(
   MaterialAttributes
   Base Color=Multiply(
    A=Multiply(A=@n1.RGB, B=@n1.RGB)
    B=MaterialFunctionCall(TextureObject=@n2, TextureSize=@n3) {materialFunction=&WorldAlignedTexture}.XYZ Texture
   )
   Roughness=Constant() {r=1}
   Normal=VertexNormalWS()
   World Position Offset=MaterialFunctionCall() {materialFunction=/Water/Caustics/Materials/Functions/CausticsBaseMeshPreview_VS}.WPO
  ) {attributeSetTypes [
    #1
    #2
    #3
    #4
   ]}
  MaterialFunctionCall!(Water Z, Additional Wet Mask, In=@n4) {materialFunction=/Water/Materials/Functions/Wetness}
 ]
}
