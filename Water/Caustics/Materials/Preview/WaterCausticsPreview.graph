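# Two-sided Single Layer Water material for previewing water caustics, with Nanite and instanced mesh support.
# Gerstner waves, driven by freezable collection time, feed a distance-blended textured surface and wave attributes over the underside function.
# Static switches select analytical Gerstner, SubUV-animated or static caustics to tint ColorScaleBehindWater.
# Scattering and absorption are zeroed on back faces via TwoSidedSign, with blurred non-analytical caustics modulating absorption.
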Material /Water/Caustics/Materials/Preview/WaterCausticsPreview
graph {
 properties {
  bUseMaterialAttributes=true
  bUsedWithInstancedStaticMeshes=true
  bUsedWithNanite=true
  pixelDepthOffsetMode=PDOM_Legacy
  refractionMethod=RM_PixelNormalOffset
  shadingModel=MSM_SingleLayerWater
  twoSided=true
  usedShadingModels=SingleLayerWater
 }
 nodes [
  n1 = LinearInterpolate(
   A=Time()
   B=CollectionParameter() {collection=/Water/Materials/MPC/MPC_Water, parameterName=ManualTime}
   Alpha=CollectionParameter() {collection=&MPC_Water, parameterName=FreezeTime}
  )
  n2 = ScalarParameter() {defaultValue=25000, parameterName=Water Distant Normal Length}
  n3 = ScalarParameter() {defaultValue=5000, parameterName=Water Distant Normal Offset}
  MaterialFunctionCall!(
   Material Atttributes=MaterialFunctionCall!(Attributes=MaterialFunctionCall() {materialFunction=/Water/Materials/Layers/WaterAttributes}.Result, Scattering, Absorption) {materialFunction=/Water/Materials/Functions/Water_Underside}.Result
   Wave Attributes=MaterialFunctionCall!(
    Wave Attributes=MaterialFunctionCall!(WaveParameterRT, Override Position, Time=@n1) {materialFunction=/Water/Caustics/Materials/Functions/GetGerstnerWaves}.Wave Attributes
    Time
    Distant Scale
    Near Scale
    Blend Alpha=MaterialFunctionCall(Fade Length=@n2, Fade Offset=@n3) {materialFunction=&CameraDepthFade}.Result
    Speed
    Far Normal Strength
    Near Normal Strength
    Far Displacement
   ) {materialFunction=/Water/Materials/Functions/WaterTextureSurface}.Wave Attributes
  ) {materialFunction=/Water/Materials/Functions/SetWaveAttributes}
  n4 = VectorParameter() {parameterName=Scattering, defaultValue=<rgba 1 1 1 0.5>}
  n5 = Divide(A=Divide(A=Multiply(A=@n4.RGB, B=@n4.RGB)) {constB=1000}, B=ScalarParameter() {defaultValue=1, parameterName=Caustics Multiplier})
  n6 = ScalarParameter() {defaultValue=1, parameterName=dst}
  n7 = Saturate(None=TwoSidedSign())
  n8 = VectorParameter() {parameterName=Absorption, defaultValue=<rgba 25 5 2 2500>}
  n9 = Divide(A=@n8.RGB, B=@n8.RGB)
  n10 = Constant()
  n11 = GetMaterialAttributes(None=MaterialFunctionCall() {materialFunction=&WaterAttributes}.Result) {attributeGetTypes [#1]}
  n12 = ScalarParameter() {defaultValue=1, parameterName=Caustics Boost}
  n13 = MaterialFunctionCall!(Refraction=@n11.Refraction, Tiling=ScalarParameter() {defaultValue=4096, parameterName=Caustics Scale}, Caustics Brightness=@n12) {materialFunction=/Water/Caustics/Materials/Functions/WaterCaustics_SubUVAnimation}
  n14 = MaterialFunctionCall!(
   Refraction=GetMaterialAttributes(None=MaterialFunctionCall() {materialFunction=&WaterAttributes}.Result) {attributeGetTypes [#1]}.Refraction
   Tiling=ScalarParameter() {defaultValue=4096, parameterName=Caustics Scale}
   Caustics Brightness=ScalarParameter() {defaultValue=1, parameterName=Caustics Boost}
  ) {materialFunction=/Water/Caustics/Materials/Functions/WaterCaustics_Static}
  n15 = StaticBoolParameter() {parameterName=Use SubUV Animation}
  n16 = StaticSwitch(True=@n13.Blurred Caustics, False=@n14.Blurred Caustics, Value=@n15)
  n17 = StaticBoolParameter() {parameterName=Use Analytical Caustics}
  n18 = Add(A=Multiply(A=ScalarParameter() {parameterName=Caustics Scattering}, B=StaticSwitch(True=@n10, False=@n16, Value=@n17)))
  n19 = StaticSwitch(
   True=MaterialFunctionCall!(Refraction=@n11.Refraction, Caustics Brightness=@n12) {materialFunction=/Water/Caustics/Materials/Functions/WaterCaustics_GerstnerWaves_PreviewOnly}.Caustics
   False=StaticSwitch(True=@n13.Caustics, False=@n14.Caustics, Value=@n15)
   Value=@n17
  )
  n20 = Multiply(A=ScalarParameter() {defaultValue=1, parameterName=Caustics Alpha}, B=ScalarParameter() {defaultValue=1, parameterName=Caustics Alpha BP})
  SingleLayerWaterMaterialOutput!(
   ScatteringCoefficients=Multiply(A=Divide(A=@n5, B=@n6), B=@n7)
   AbsorptionCoefficients=Multiply(A=Divide(A=@n9, B=@n18), B=@n7)
   PhaseG=ScalarParameter() {defaultValue=0.1, parameterName=Anisotropy}
   ColorScaleBehindWater=StaticSwitchParameter(True=LinearInterpolate(B=@n19, Alpha=@n20) {constA=1}, False=Constant() {r=1}) {defaultValue=true, parameterName=Caustics}
  )
 ]
}
